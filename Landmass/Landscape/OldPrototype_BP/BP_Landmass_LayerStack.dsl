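; Prototype Landmass editor actor that procedurally builds landscape height and weightmaps from seeded noise, imported landscape height and a stack of spline brushes.
; Construction sizes render targets, configures MIDs and MPC_Erosion, then refreshes height shown on an instanced preview mesh.
; Drawing mode spawns landmass, mountain, road, river or weightmap brushes, using editor tick for cursor previews.
; Scene captures supply mesh depth, while jump-flood and erosion components share ping-pong targets, with height and normals exportable.
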(event Custom|Rebind%20Preview%20Mesh%20RT
  (Rendering|Material|SetTextureParameterValue (Variables|7%29Debug|MIDs|GetMeshPreviewMID) "HeightRT" (Utilities|HeightPingPong)))

(event Custom|Export%20Height
  (bind _source_rt (Utilities|HeightPingPong))
  (Rendering|Material|SetTextureParameterValue (Variables|7%29Debug|MIDs|GetNormalizeHeightMID) "RT" _source_rt)
  (Rendering|ExportRenderTarget _source_rt "c:\\RToutput\\" "01.png"))

(event Custom|Cache%20Normal
  (bind _compute_normal_mid (Variables|7%29Debug|MIDs|GetCompute%20Normal%20MID))
  (Rendering|Material|SetScalarParameterValue _compute_normal_mid "OutputLDR" 1.0 :!class /Script/Engine.MaterialInstanceDynamic)
  (Rendering|DrawMaterialtoRenderTarget "/Landmass/Landscape/OldPrototype_BP/Textures/LS_Debug_RT_01.LS_Debug_RT_01" _compute_normal_mid)
  (Rendering|Material|SetScalarParameterValue _compute_normal_mid "OutputLDR" 0.0 :!class /Script/Engine.MaterialInstanceDynamic))

(event Tick|EventCustomTick
  (params (DeltaSeconds :type Float%20%28single-precision%29))
  (bind _curshape (Variables|Default|GetCurShape))
  (bind _spline (Variables|/Landmass/Landscape/OldPrototype_BP/Brushes/BP_LandmassBrush.BP_LandmassBrush_C|GetSpline _curshape))
  (bind _returnvalue (Spline|GetNumberOfSplinePoints _spline))
  (bind _returnvalue_2 (- _returnvalue 1))
  (bind _returnvalue_3 (Spline|GetLocationatSplinePoint _spline _returnvalue_2 "World"))
  (bind _curshape_2 (Variables|Default|GetCurShape))
  (bind _starting_z_elevation (Variables|5%29ShapeDrawing|GetStarting%20Z%20Elevation))
  (if (Variables|1%29RealTimeSimulation|GetAlways%20Tick%20for%20Perf%20Test)
    (2%29Generation|UpdateHeight)
    (elif (Variables|7%29Debug|GetTicking%20Erosion)
      (if (or (< (Variables|7%29Debug|GetStackHeightIdx) (Variables|7%29Debug|GetMax%20Iterations)) (== (Variables|7%29Debug|GetMax%20Iterations) 0))
        (Class|ErosionComponent|ErosionTick (Variables|Default|GetLandscapeErosionComponent)))
      (elif (Variables|Default|GetDrawingMode)
        ; Debug Lines
        (Rendering|Debug|FlushPersistentDebugLines)
        (if (> (Spline|GetNumberOfSplinePoints (Variables|/Landmass/Landscape/OldPrototype_BP/Brushes/BP_LandmassBrush.BP_LandmassBrush_C|GetSpline (Variables|Default|GetCurShape))) 0)
          (bind (_cameralocation _rayorigin _raydirection) (Rendering|GetCursorWorldRay :!class /Script/LandmassEditor.LandmassBlueprintFunctionLibrary))
          (Rendering|Debug|DrawDebugLine _returnvalue_3 (+ _rayorigin (* _raydirection (/ (- (.z _rayorigin) _starting_z_elevation) (Math|Float|Absolute(Float) (.z _raydirection))))) "(R=0.000000,G=0.416800,B=1.000000,A=1.000000)" 0.0 2.5e+02)
          (Rendering|Material|SetVectorParameterValue "/Landmass/Landscape/OldPrototype_BP/Mats/MPC/MPC_Drawing.MPC_Drawing" "P0" (Math|Conversions|ToLinearColor(Vector) _returnvalue_3) :!class /Script/Engine.KismetMaterialLibrary)
          (Rendering|Material|SetVectorParameterValue "/Landmass/Landscape/OldPrototype_BP/Mats/MPC/MPC_Drawing.MPC_Drawing" "P1" (Math|Conversions|ToLinearColor(Vector) (+ _rayorigin (* _raydirection (/ (- (.z _rayorigin) _starting_z_elevation) (Math|Float|Absolute(Float) (.z _raydirection)))))) :!class /Script/Engine.KismetMaterialLibrary)
          (Rendering|Material|SetVectorParameterValue "/Landmass/Landscape/OldPrototype_BP/Mats/MPC/MPC_Drawing.MPC_Drawing" "Cursor" (Math|Conversions|ToLinearColor(Vector) (+ _rayorigin (* _raydirection (/ (- (.z _rayorigin) _starting_z_elevation) (Math|Float|Absolute(Float) (.z _raydirection)))))) :!class /Script/Engine.KismetMaterialLibrary)
          (if (> (Spline|GetNumberOfSplinePoints (Variables|/Landmass/Landscape/OldPrototype_BP/Brushes/BP_LandmassBrush.BP_LandmassBrush_C|GetSpline (Variables|Default|GetCurShape))) 1)
            (switch Utilities|FlowControl|Switch|SwitchonShapeDrawingMode (Variables|5%29ShapeDrawing|GetNew%20Shape%20Type)
              (:NewEnumerator0
                _)
              (:NewEnumerator6
                _)
              (:NewEnumerator7
                _))
            (Variables|/Landmass/Landscape/OldPrototype_BP/Brushes/BP_LandmassBrush.BP_LandmassBrush_C|SetDistanceIsCached false _curshape_2)
            (Variables|/Landmass/Landscape/OldPrototype_BP/Brushes/BP_LandmassBrush.BP_LandmassBrush_C|SetDynamicPreviewPoint (Math|Color|MakeColor (.x (+ _rayorigin (* _raydirection (/ (- (.z _rayorigin) _starting_z_elevation) (Math|Float|Absolute(Float) (.z _raydirection)))))) (.y (+ _rayorigin (* _raydirection (/ (- (.z _rayorigin) _starting_z_elevation) (Math|Float|Absolute(Float) (.z _raydirection)))))) (.z (+ _rayorigin (* _raydirection (/ (- (.z _rayorigin) _starting_z_elevation) (Math|Float|Absolute(Float) (.z _raydirection))))))) _curshape_2)
            (2%29Generation|UpdateHeight))
          (else
            (bind (_cameralocation_2 _rayorigin_2 _raydirection_2) (Rendering|GetCursorWorldRay :!class /Script/LandmassEditor.LandmassBlueprintFunctionLibrary))
            (Rendering|Debug|DrawDebugSphere (+ _rayorigin_2 (* _raydirection_2 (/ (- (.z _rayorigin_2) _starting_z_elevation) (Math|Float|Absolute(Float) (.z _raydirection_2))))) 512.0 12 "(R=0.807292,G=0.000000,B=0.005006,A=1.000000)" 5.0 1e+02)
            (Rendering|Material|SetVectorParameterValue "/Landmass/Landscape/OldPrototype_BP/Mats/MPC/MPC_Drawing.MPC_Drawing" "Cursor" (Math|Conversions|ToLinearColor(Vector) (+ _rayorigin_2 (* _raydirection_2 (/ (- (.z _rayorigin_2) _starting_z_elevation) (Math|Float|Absolute(Float) (.z _raydirection_2)))))) :!class /Script/Engine.KismetMaterialLibrary)
            (Rendering|Material|SetVectorParameterValue "/Landmass/Landscape/OldPrototype_BP/Mats/MPC/MPC_Drawing.MPC_Drawing" "P0" (Math|Conversions|ToLinearColor(Vector) (+ _rayorigin_2 (* _raydirection_2 (/ (- (.z _rayorigin_2) _starting_z_elevation) (Math|Float|Absolute(Float) (.z _raydirection_2)))))) :!class /Script/Engine.KismetMaterialLibrary)
            (Rendering|Material|SetVectorParameterValue "/Landmass/Landscape/OldPrototype_BP/Mats/MPC/MPC_Drawing.MPC_Drawing" "P1" (Math|Conversions|ToLinearColor(Vector) (+ _rayorigin_2 (* _raydirection_2 (/ (- (.z _rayorigin_2) _starting_z_elevation) (Math|Float|Absolute(Float) (.z _raydirection_2)))))) :!class /Script/Engine.KismetMaterialLibrary)))
        (else
          (6%29MeshUtils|SpawnInstanceMeshes)
          (Default|SetEditorTickEnabled false))))))

(event Custom|Print%20Path
  (Development|PrintString (Utilities|GetObjectPathString (Variables|2%29Generation|GetMesh%20Preview%20Material))))

(fn 1%29CreateShape
  (bind _curshape (Variables|Default|GetCurShape))
  (bind _spline (Variables|/Landmass/Landscape/OldPrototype_BP/Brushes/BP_LandmassBrush.BP_LandmassBrush_C|GetSpline _curshape))
  (bind _returnvalue (Spline|GetNumberOfSplinePoints _spline))
  (bind _starting_z_elevation (Variables|5%29ShapeDrawing|GetStarting%20Z%20Elevation))
  (bind _returnvalue_2 (+ _starting_z_elevation 0.015))
  (bind _curshape_2 (Variables|Default|GetCurShape))
  (bind _curshape_3 (Variables|Default|GetCurShape))
  (if (Variables|7%29Debug|GetTicking%20Erosion)
    (Variables|7%29Debug|SetTicking%20Erosion false)
    (Development|PrintString "Erosion Simulation Aborted by Drawing Shape!"))
  (if (Variables|Default|GetDrawingMode)
    (bind (_cameralocation _rayorigin) (Rendering|GetCursorWorldRay :!class /Script/LandmassEditor.LandmassBlueprintFunctionLibrary))
    (Spline|AddSplinePoint _spline (+ _cameralocation (* _rayorigin (/ (- (.z _cameralocation) _returnvalue_2) (Math|Float|Absolute(Float) (.z _rayorigin))))) "World" true)
    (if (Math|Boolean|ORBoolean (Utilities|Enum|Equal(Enum) (Variables|5%29ShapeDrawing|GetNew%20Shape%20Type) "NewEnumerator0") (Utilities|Enum|Equal(Enum) (Variables|5%29ShapeDrawing|GetNew%20Shape%20Type) "NewEnumerator6") (Utilities|Enum|Equal(Enum) (Variables|5%29ShapeDrawing|GetNew%20Shape%20Type) "NewEnumerator7"))
      (Spline|SetTangentsatSplinePoint _spline (- _returnvalue 1) (Math|Vector|MakeVector 0.0) (Math|Vector|MakeVector 0.0) "World" true))
    (if (> _returnvalue 2)
      (switch Utilities|FlowControl|Switch|SwitchonShapeDrawingMode (Variables|5%29ShapeDrawing|GetNew%20Shape%20Type)
        (:NewEnumerator0
          _)
        (:NewEnumerator6
          _)
        (:NewEnumerator1
          (Class|BPLandmassBrush|Constructor _curshape_3 :!label _label) ; Run Constructor to draw new Spline Meshes
          (Variables|/Landmass/Landscape/OldPrototype_BP/Brushes/BP_LandmassBrush.BP_LandmassBrush_C|SetDistanceIsCached false _curshape_3 :!label _label_2)
          (2%29Generation|UpdateHeight))
        (:NewEnumerator5
          (goto _label))
        (:NewEnumerator7
          _))
      (Class|BPLandmassBrush|CanvasDraw _curshape_3)
      (goto _label_2))
    (else
      (Rendering|ClearRenderTarget2D (Variables|7%29Debug|RenderTargets|GetJump%20Flood%20Target%20B_Vel))
      (Rendering|ClearRenderTarget2D (Variables|7%29Debug|RenderTargets|GetJump%20Flood%20Target%20A_Vel))
      (Rendering|ClearRenderTarget2D (Variables|7%29Debug|RenderTargets|GetMesh%20Depth%20RT))
      (Rendering|ClearRenderTarget2D (Utilities|HeightPingPong))
      (Rendering|Material|SetTextureParameterValue (Variables|7%29Debug|MIDs|GetDebugDrawMID) "Shape" (Variables|7%29Debug|RenderTargets|GetMesh%20Depth%20RT))
      (Rendering|Material|SetMaterial (Variables|Default|GetDrawingDebug) 0 (Variables|7%29Debug|MIDs|GetDebugDrawMID))
      (Variables|Default|SetDrawingMode true)
      (Rendering|SetVisibility (Variables|Default|GetDrawingDebug) true false)
      (Default|SetEditorTickEnabled true)
      (bind _returnvalue_3 (Game|SpawnActorfromClass :Class "/Landmass/Landscape/OldPrototype_BP/Brushes/BP_LandmassBrush.BP_LandmassBrush_C" :SpawnTransform (Math|Transform|MakeTransform (Math|Vector|MakeVector 0.0 0.0 _returnvalue_2)) :CollisionHandlingOverride "Undefined" :TransformScaleMethod "MultiplyWithRoot" :Brush%20Type (Utilities|Select :Index (Variables|5%29ShapeDrawing|GetNew%20Shape%20Type) :NewEnumerator0 "NewEnumerator0" :NewEnumerator6 "NewEnumerator0" :NewEnumerator1 "NewEnumerator1" :NewEnumerator5 "NewEnumerator1" :NewEnumerator7 "NewEnumerator0") :BlendMode "NewEnumerator0" :Falloff%20Width (Utilities|Select :Index (Variables|5%29ShapeDrawing|GetNew%20Shape%20Type) :NewEnumerator0 8192.0 :NewEnumerator6 8192.0 :NewEnumerator1 2048.0 :NewEnumerator5 2048.0 :NewEnumerator7 8192.0) :Falloff%20Angle (Utilities|Select :Index (Variables|5%29ShapeDrawing|GetNew%20Shape%20Type) :NewEnumerator0 0.0 :NewEnumerator6 45.0 :NewEnumerator1 45.0 :NewEnumerator5 45.0 :NewEnumerator7 0.0) :Edge%20Width%20Offset 0.0 :UseMeshDepth (Utilities|Select :Index (Variables|5%29ShapeDrawing|GetNew%20Shape%20Type) :NewEnumerator0 true :NewEnumerator6 false :NewEnumerator1 true :NewEnumerator5 true :NewEnumerator7 true) :Curl%201%20Strength (Utilities|Select :Index (Variables|5%29ShapeDrawing|GetNew%20Shape%20Type) :NewEnumerator0 0.5 :NewEnumerator6 0.5 :NewEnumerator1 0.0 :NewEnumerator5 0.0 :NewEnumerator7 0.5) :Curl%202%20Strength (Utilities|Select :Index (Variables|5%29ShapeDrawing|GetNew%20Shape%20Type) :NewEnumerator0 0.1 :NewEnumerator6 0.1 :NewEnumerator1 0.0 :NewEnumerator5 0.0 :NewEnumerator7 0.1) :Default%20Material "/Landmass/Landscape/OldPrototype_BP/Mats/Brushes/M_Foundation.M_Foundation" :Custom%20Override%20Mat "/Landmass/Landscape/OldPrototype_BP/Mats/Brushes/M_Foundation.M_Foundation" :Use%20Override%20Mat false :Z%20Offset (Utilities|Select :Index (Variables|5%29ShapeDrawing|GetNew%20Shape%20Type) :NewEnumerator0 0.0 :NewEnumerator6 0.0 :NewEnumerator1 -16.0 :NewEnumerator5 -16.0 :NewEnumerator7 0.0) :Affect%20Heightmap (not (Utilities|Enum|Equal(Enum) (Variables|5%29ShapeDrawing|GetNew%20Shape%20Type) "NewEnumerator7")) :Spline%20Mesh (Utilities|Select :Index (Variables|5%29ShapeDrawing|GetNew%20Shape%20Type) :NewEnumerator0 "/Engine/EditorLandscapeResources/SplineEditorMesh.SplineEditorMesh" :NewEnumerator6 "/Engine/EditorLandscapeResources/SplineEditorMesh.SplineEditorMesh" :NewEnumerator1 "/Engine/EditorLandscapeResources/SplineEditorMesh.SplineEditorMesh" :NewEnumerator5 "/Engine/EditorLandscapeResources/SplineEditorMesh.SplineEditorMesh") :Spline%20Material (Utilities|Select :Index (Variables|5%29ShapeDrawing|GetNew%20Shape%20Type) :NewEnumerator1 "/Landmass/PreviewContent/Roads/Mats/M_RoadSpline_01.M_RoadSpline_01" :NewEnumerator5 "/Landmass/PreviewContent/Water/Mats/M_Simple_River_01.M_Simple_River_01") :Blur%20Distance%20Field false :Flatten%20Z false :Blur%20Filter%20Scale 1.0))
      (bind _output_get (Variables|Default|SetCurShape _returnvalue_3))
      (Class|BPLandmassBrush|Constructor _output_get)
      (Spline|ClearSplinePoints (Variables|/Landmass/Landscape/OldPrototype_BP/Brushes/BP_LandmassBrush.BP_LandmassBrush_C|GetSpline _curshape_2) true)
      (EditorScripting|ActorEditing|SetActorLabel _curshape_2 (Utilities|String|Append "BP_LandscapeBrush_" (Utilities|Select :Index (Variables|5%29ShapeDrawing|GetNew%20Shape%20Type) :NewEnumerator0 "Landmass" :NewEnumerator6 "Mountain" :NewEnumerator1 "RoadSpline" :NewEnumerator5 "RiverSpline" :NewEnumerator7 "Weightmap")) true)
      (if (Utilities|Enum|Equal(Enum) (Variables|5%29ShapeDrawing|GetNew%20Shape%20Type) "NewEnumerator7")
        (Utilities|Array|Add (Variables|/Landmass/Landscape/OldPrototype_BP/Brushes/BP_LandmassBrush.BP_LandmassBrush_C|GetWeightmap%20Target%20Settings (Variables|Default|GetCurShape)) (Utilities|Struct|MakeBrushWM (Variables|5%29ShapeDrawing|WeightmapShape|GetWeightmap%20Name)))))))

(fn 1%29ReadLandscapeHeight
  (Rendering|Material|SetScalarParameterValue "/Landmass/Landscape/BlueprintBrushes/MPC/MPC_Erosion.MPC_Erosion" "ShowProcedural" :!class /Script/Engine.KismetMaterialLibrary)
  (Rendering|ClearRenderTarget2D (Variables|7%29Debug|RenderTargets|GetLandscape%20Height%20RT))
  (Rendering|LandscapeExportHeightmaptoRenderTarget (Variables|3%29Landscape|GetLandscape) (Variables|7%29Debug|RenderTargets|GetLandscape%20Height%20RT) true true)
  (2%29Generation|UpdateHeight)
  (Rendering|Material|SetScalarParameterValue "/Landmass/Landscape/BlueprintBrushes/MPC/MPC_Erosion.MPC_Erosion" "ShowProcedural" (Math|Conversions|ToFloat(Boolean) (Variables|3%29Landscape|GetShow%20Procedural%20Impact)) :!class /Script/Engine.KismetMaterialLibrary))

(fn 2%29CompleteShape
  (Variables|Default|SetDrawingMode false)
  (Class|BPLandmassBrush|CenterActorOnSplinePoints (Variables|Default|GetCurShape))
  (Variables|Default|SetCurShape 0)
  (Default|SetEditorTickEnabled false) ; Allows updating line previews while drawing
  (Rendering|SetVisibility (Variables|Default|GetDrawingDebug) false false))

(fn 2%29CopyLandscapeRange
  (bind _landscape (Variables|3%29Landscape|GetLandscape))
  (Utilities|IsValid _landscape
    (:Is%20Valid
      (Variables|3%29Landscape|SetTotal%20Z%20Range%20Kilometers (* (/ (.z (.scale (Transformation|GetActorTransform _landscape))) 128.0) 0.65536)))))

(fn CaptureMeshDepth
  (params (Primitive :type Primitive%20Component%20Object%20Reference))
  (bind _meshscenecapture (Variables|Default|GetMeshSceneCapture))
  (Rendering|SceneCapture|ClearShowOnlyComponents _meshscenecapture)
  (Rendering|SceneCapture|ShowOnlyActorComponents _meshscenecapture (Components|GetOwner Primitive) false)
  (Variables|/Script/Engine.SceneCaptureComponent2D|SetTextureTarget (Variables|7%29Debug|RenderTargets|GetMesh%20Depth%20RT) (Variables|Default|GetMeshSceneCapture))
  (Rendering|SceneCapture|CaptureScene (Variables|Default|GetMeshSceneCapture) :!class /Script/Engine.SceneCaptureComponent2D))

(fn CaptureSplineDepth
  (params (Primitive :type Array%20of%20Spline%20Mesh%20Component%20Object%20References))
  (Rendering|SceneCapture|ClearShowOnlyComponents (Variables|Default|GetMeshSceneCapture))
  (for _array_element Primitive
    (Rendering|SceneCapture|ShowOnlyComponent (Variables|Default|GetMeshSceneCapture) _array_element))
  (Rendering|SceneCapture|CaptureScene (Variables|Default|GetMeshSceneCapture) :!class /Script/Engine.SceneCaptureComponent2D))

(fn ConfigRes
  (bind _sections_x (Variables|3%29Landscape|GetSections%20X))
  (bind _returnvalue (Math|Conversions|ToFloat(Integer) _sections_x))
  (bind _sections_y (Variables|3%29Landscape|GetSections%20Y))
  (bind _returnvalue_2 (Math|Conversions|ToFloat(Integer) _sections_y))
  (bind _component_size (Variables|3%29Landscape|GetComponent%20Size))
  (bind _returnvalue_3 (Utilities|Select :Index _component_size :NewEnumerator0 127.0 :NewEnumerator1 255.0))
  (bind _returnvalue_4 (* _returnvalue_2 _returnvalue_3))
  (bind _returnvalue_5 (+ _returnvalue_4 1.0))
  (bind _returnvalue_6 (* _returnvalue _returnvalue_3))
  (bind _returnvalue_7 (+ _returnvalue_6 1.0))
  (bind _returnvalue_8 (Math|Vector2D|MakeVector2D _returnvalue_7 _returnvalue_5))
  (bind (_x _y) (Math|Vector2D|BreakVector2D (Variables|3%29Landscape|GetOverall%20Resolution)))
  (bind (_x_2 _y_2) (Math|Vector2D|BreakVector2D (Variables|7%29Debug|GetTotalSize)))
  (bind (_x_3 _y_3) (Math|Vector2D|BreakVector2D _returnvalue_8))
  (bind _drawingdebug (Variables|Default|GetDrawingDebug))
  (Variables|3%29Landscape|SetSections%20X (Math|Integer|Min(Integer) (Variables|3%29Landscape|GetSections%20X) 16))
  (Variables|3%29Landscape|SetSections%20Y (Math|Integer|Min(Integer) (Variables|3%29Landscape|GetSections%20Y) 16))
  (Variables|3%29Landscape|SetOverall%20Resolution (select (Variables|3%29Landscape|Overrides|GetForce%20Power%20of%202) (Math|Vector2D|MakeVector2D (Math|Float|Power 2.0 (Math|Conversions|ToFloat(Integer) (Math|Float|Ceil (Math|Float|Log _x_3 2.0)))) (Math|Float|Power 2.0 (Math|Conversions|ToFloat(Integer) (Math|Float|Ceil (Math|Float|Log _y_3 2.0))))) _returnvalue_8))
  (bind _output_get (Variables|7%29Debug|SetSizePerQuad (select (== (Variables|3%29Landscape|Overrides|GetOverride%20Size) 0.0) 1e+02 (/ (Variables|3%29Landscape|Overrides|GetOverride%20Size) (- (Math|Float|Max(Float) _x _y) 1.0)))))
  (Variables|7%29Debug|SetTotalSize (* (- _returnvalue_8 1.0) _output_get))
  (Transformation|SetRelativeLocation (Variables|Default|GetActor%20Ticking) (Math|Vector|MakeVector (+ (* (Math|Vector2D|BreakVector2D (Variables|7%29Debug|GetTotalSize)) -0.5) -5e+03)) false false)
  (Variables|3%29Landscape|SetLandscape%20Actor%20Z%20Scale (/ (* (Variables|3%29Landscape|GetTotal%20Z%20Range%20Kilometers) 1e+05) 512.0))
  (Variables|7%29Debug|SetRaw%20Z%20Scale (/ (Variables|3%29Landscape|GetTotal%20Z%20Range%20Kilometers) 0.65536))
  (Transformation|SetWorldScale3D _drawingdebug (Math|Vector|MakeVector _x_2 _y_2 1.0))
  (Transformation|SetWorldLocation _drawingdebug (Math|Vector|MakeVector (.x (Transformation|GetActorLocation)) (.y (Transformation|GetActorLocation)) (Variables|5%29ShapeDrawing|GetStarting%20Z%20Elevation)) false false))

(fn Constructor
  (bind (_x _y) (Math|Vector2D|BreakVector2D (Variables|7%29Debug|GetTotalSize)))
  (bind (_x_2 _y_2) (Math|Vector2D|BreakVector2D (Variables|7%29Debug|GetTotalSize)))
  (bind _layer_stack (Variables|6%29BrushLayerStack|GetLayer%20Stack))
  (bind _landscape_ismc (Variables|Default|GetLandscape%20ISMC))
  (bind _scattermeshscenecapture (Variables|Default|GetScatterMeshSceneCapture))
  (CallFunction|ConfigRes)
  (CallFunction|CreateRenderTargets)
  (CallFunction|CreateMIDs)
  (CallFunction|SetMIDParams)
  (CallFunction|SetMPCParams)
  (Utilities|FlowControl|Sequence
    (:then_0
      (Variables|/Script/Engine.SceneCaptureComponent2D|SetOrthoWidth (+ (Math|Float|Max(Float) _x_2 _y_2) (Variables|7%29Debug|GetSizePerQuad)) (Variables|Default|GetMeshSceneCapture))
      (Transformation|SetWorldLocation (Variables|Default|GetMeshSceneCapture) (Math|Vector|MakeVector 0.0 0.0 5e+04) false false)
      (Variables|/Script/Engine.SceneCaptureComponent2D|SetTextureTarget (Variables|7%29Debug|RenderTargets|GetMesh%20Depth%20RT) _scattermeshscenecapture)
      (Variables|/Script/Engine.SceneCaptureComponent2D|SetOrthoWidth (+ (Math|Float|Max(Float) _x _y) (Variables|7%29Debug|GetSizePerQuad)) _scattermeshscenecapture)
      (Rendering|SceneCapture|ShowOnlyComponent (Variables|Default|GetScatterMeshSceneCapture) (Variables|Default|GetScatter%20Particles%20ISMC)))
    (:then_1
      (Class|JumpFloodComponent2D|AssignRenderTargets (Variables|Default|GetJumpFlood_Component2D) (Variables|7%29Debug|RenderTargets|GetJump%20Flood%20Target%20A_Vel) (Variables|7%29Debug|RenderTargets|GetJump%20Flood%20Target%20B_Vel) :!class /Landmass/DistanceFields/Blueprints/JumpFlood_Component2D.JumpFlood_Component2D_C)
      (Components|StaticMesh|SetStaticMesh _landscape_ismc (Utilities|Select :Index (Variables|3%29Landscape|GetComponent%20Size) :NewEnumerator0 "/Landmass/Landscape/Meshes/S_Preview_LS_128x128.S_Preview_LS_128x128" :NewEnumerator1 "/Landmass/Landscape/Meshes/S_Preview_LS_256x256.S_Preview_LS_256x256"))
      (Components|StaticMesh|SetStaticMesh (Variables|Default|GetScatter%20Particles%20ISMC) (Utilities|Select :Index (Variables|3%29Landscape|GetComponent%20Size) :NewEnumerator0 "/Landmass/Landscape/Meshes/S_Scatter_127x127_quads.S_Scatter_127x127_quads" :NewEnumerator1 "/Landmass/Landscape/Meshes/S_Scatter_255x255_quads.S_Scatter_255x255_quads"))
      (Rendering|Material|SetMaterial _landscape_ismc 0 (Variables|7%29Debug|MIDs|GetMeshPreviewMID))
      (Rendering|SetVisibility _landscape_ismc (Variables|2%29Generation|GetShow%20Preview%20Mesh) false)
      (Rendering|SetVisibility (Variables|Default|GetScatter%20Particles%20ISMC) (Variables|1%29RealTimeSimulation|GetShow%20Water) false)
      (Class|ErosionComponent|Constructor (Variables|Default|GetLandscapeErosionComponent) (Utilities|Struct|MakeErosionInterface (Variables|7%29Debug|RenderTargets|GetRender%20Target%20A) (Variables|7%29Debug|RenderTargets|GetRender%20Target%20B) (Variables|7%29Debug|RenderTargets|GetJump%20Flood%20Target%20A_Vel) (Variables|7%29Debug|RenderTargets|GetJump%20Flood%20Target%20B_Vel) (Variables|7%29Debug|RenderTargets|GetMesh%20Depth%20RT) (Variables|Default|GetMeshSceneCapture) (Variables|7%29Debug|RenderTargets|GetPosition%20A) (Variables|7%29Debug|RenderTargets|GetPosition%20B)))
      (Rendering|SceneCapture|CaptureScene (Variables|Default|GetMeshSceneCapture) :!class /Script/Engine.SceneCaptureComponent2D)
      (Variables|/Script/Engine.SceneCaptureComponent2D|SetTextureTarget (Variables|7%29Debug|RenderTargets|GetMesh%20Depth%20RT) (Variables|Default|GetMeshSceneCapture))
      (Rendering|LOD|SetForcedLodModel (Variables|Default|GetLandscape%20ISMC) (Variables|3%29Landscape|Overrides|GetForcedLOD)))
    (:then_2
      (if (Variables|7%29Debug|GetTicking%20Erosion)
        _
        (elif (or (Utilities|CheckSizeParamsChanged) (not (== (Variables|7%29Debug|CachedValues|GetStoredInstances) (* (Variables|3%29Landscape|GetSections%20X) (Variables|3%29Landscape|GetSections%20Y)))))
          (Default|SetEditorTickEnabled true) ; A hack to allow a latent update to ISMC, allowing instances to persist between construction runs
          (CallFunction|InvalidateBrushCache)
          (elif (Default|CheckSeedParamsChanged)
            _
            (else
              (CallFunction|UpdateNormalandPreviewRTAssignments (Variables|7%29Debug|RenderTargets|GetJump%20Flood%20Target%20A_Vel))
              (Utilities|FlowControl|Sequence :!label _label
                (:then_0
                  (Utilities|CacheSizeParams))
                (:then_1
                  (Default|CacheSeedParams))
                (:then_2
                  (for (_array_element _array_index) _layer_stack
                    (Utilities|IsValid _array_element
                      (:Is%20Not%20Valid
                        (Utilities|Array|RemoveIndex _layer_stack _array_index))))))))))
      (2%29Generation|UpdateHeight)
      (goto _label))))

(fn CreateRenderTargets
  (bind _overall_resolution (Variables|3%29Landscape|GetOverall%20Resolution))
  (bind _weightmaps (Variables|3%29Landscape|WeightmapSetup|GetWeightmaps))
  (Utilities|FlowControl|Sequence
    (:then_0
      (Utilities|CreateRTIfNullOrChanged (Variables|7%29Debug|RenderTargets|GetRender%20Target%20A) "RTF_RGBA32f"))
    (:then_1
      (Utilities|CreateRTIfNullOrChanged (Variables|7%29Debug|RenderTargets|GetRender%20Target%20B) "RTF_RGBA32f"))
    (:then_2
      (Utilities|CreateRTIfNullOrChanged (Variables|7%29Debug|RenderTargets|GetNormal) "RTF_RGBA16f"))
    (:then_3
      (Utilities|CreateRTIfNullOrChanged (Variables|7%29Debug|RenderTargets|GetJump%20Flood%20Target%20A_Vel) "RTF_RGBA32f"))
    (:then_4
      (Utilities|CreateRTIfNullOrChanged (Variables|7%29Debug|RenderTargets|GetJump%20Flood%20Target%20B_Vel) "RTF_RGBA32f"))
    (:then_5
      (Utilities|CreateRTIfNullOrChanged (Variables|7%29Debug|RenderTargets|GetNoise%20Seed%20%20RT) "RTF_RG8"))
    (:then_6
      (Utilities|CreateRTIfNullOrChanged (Variables|7%29Debug|RenderTargets|GetLandscape%20Height%20RT) "RTF_RG8"))
    (:then_7
      (Utilities|CreateRTIfNullOrChanged (Variables|7%29Debug|RenderTargets|GetMesh%20Depth%20RT) "RTF_RGBA32f"))
    (:then_8
      (Utilities|FlowControl|Sequence
        (:then_0
          (if (Utilities|Enum|Equal(Enum) (Variables|4%29Erosion|GetErosion%20Mode) "NewEnumerator0")
            (Utilities|FlowControl|Sequence
              (:then_0
                (Utilities|CreateRTIfNullOrChanged (Variables|7%29Debug|RenderTargets|GetPosition%20A) "RTF_RGBA32f"))
              (:then_1
                (Utilities|CreateRTIfNullOrChanged (Variables|7%29Debug|RenderTargets|GetPosition%20B) "RTF_RGBA32f")))
            (else
              (Variables|7%29Debug|RenderTargets|SetPosition%20A "")
              (Variables|7%29Debug|RenderTargets|SetPosition%20B ""))))
        (:then_1
          ; Allocate User Defined Weightmaps
          (bind _values (Utilities|Map|Values (Variables|3%29Landscape|WeightmapSetup|GetWeightmaps)))
          (for (_array_element _array_index) _values
            (Utilities|IsValid (Utilities|Struct|BreakRenderTargetChannels _array_element)
              (:Is%20Valid
                (if (and (== (Variables|/Script/Engine.TextureRenderTarget2D|GetSizeX (Utilities|Struct|BreakRenderTargetChannels _array_element)) (Math|Float|Truncate (Math|Vector2D|BreakVector2D (/ _overall_resolution (Math|Conversions|ToFloat(Integer) (.ResolutionDivisor_20_132481E24944E807ABE34DBFCE2F0966 (Utilities|Struct|BreakRenderTargetChannels _array_element))))))) (== (Variables|/Script/Engine.TextureRenderTarget2D|GetSizeY (Utilities|Struct|BreakRenderTargetChannels _array_element)) (Math|Float|Truncate (.Y (Math|Vector2D|BreakVector2D (/ _overall_resolution (Math|Conversions|ToFloat(Integer) (.ResolutionDivisor_20_132481E24944E807ABE34DBFCE2F0966 (Utilities|Struct|BreakRenderTargetChannels _array_element)))))))))
                  _
                  (else
                    _)))
              (:Is%20Not%20Valid
                _))
            (bind _returnvalue (Rendering|CreateRenderTarget2D (Math|Float|Truncate (Math|Vector2D|BreakVector2D (/ _overall_resolution (Math|Conversions|ToFloat(Integer) (.ResolutionDivisor_20_132481E24944E807ABE34DBFCE2F0966 (Utilities|Struct|BreakRenderTargetChannels _array_element)))))) (Math|Float|Truncate (.Y (Math|Vector2D|BreakVector2D (/ _overall_resolution (Math|Conversions|ToFloat(Integer) (.ResolutionDivisor_20_132481E24944E807ABE34DBFCE2F0966 (Utilities|Struct|BreakRenderTargetChannels _array_element)))))))))
            (bind _keys (Utilities|Map|Keys _weightmaps))
            (Utilities|Map|Add _weightmaps (Utilities|Array|Get(aref) _keys _array_index) (Utilities|Struct|MakeRenderTargetChannels _returnvalue (.R_6_6FEDD63A4C392A58DB7793AFC063FE06 (Utilities|Struct|BreakRenderTargetChannels _array_element)) (.G_8_70B7D3C141576DD58ACE3E83E88865BC (Utilities|Struct|BreakRenderTargetChannels _array_element)) (.B_10_BB4928164B8A11CB3E82F8A3161F40F5 (Utilities|Struct|BreakRenderTargetChannels _array_element)) (.A_23_8074622A4536EA2CF5EEFF9B0027115F (Utilities|Struct|BreakRenderTargetChannels _array_element)) (.ResolutionDivisor_20_132481E24944E807ABE34DBFCE2F0966 (Utilities|Struct|BreakRenderTargetChannels _array_element))))))))))

(fn DrawHeighttoCBRTAsset
  (bind _normalizeheightmid (Variables|7%29Debug|MIDs|GetNormalizeHeightMID))
  (Rendering|Material|SetTextureParameterValue _normalizeheightmid "RT" (Utilities|HeightPingPong))
  (Rendering|Material|SetScalarParameterValue _normalizeheightmid "Max" (Variables|7%29Debug|GetNormalize%20Max%20Z) :!class /Script/Engine.MaterialInstanceDynamic)
  (Rendering|DrawMaterialtoRenderTarget (Variables|7%29Debug|GetRT%20to%20write%20to) (Variables|7%29Debug|MIDs|GetNormalizeHeightMID)))

(fn ExportHeightmapFile
  (Rendering|ExportRenderTarget (Utilities|HeightPingPong) (Variables|7%29Debug|GetExport%20File%20Path) (Variables|7%29Debug|GetExport%20File%20Name)))

(fn ExportNormalFile
  (Rendering|ExportRenderTarget (Variables|7%29Debug|RenderTargets|GetNormal) (Variables|7%29Debug|GetExport%20File%20Path) (Variables|7%29Debug|GetExport%20File%20Name)))

(fn GetSize
  (bind (_origin _boxextent) (Collision|GetActorBounds (Variables|3%29Landscape|GetLandscape) false false))
  (Development|PrintString (Utilities|String|ToString(Vector) _boxextent)))

(fn InvalidateBrushCache
  (Rendering|ClearRenderTarget2D (Variables|7%29Debug|RenderTargets|GetLandscape%20Height%20RT) "(R=0.501961,G=0.000000,B=0.000000,A=1.000000)")
  (for _array_element (Variables|6%29BrushLayerStack|GetLayer%20Stack)
    (Utilities|IsValid _array_element
      (:Is%20Valid
        (Variables|/Landmass/Landscape/OldPrototype_BP/Brushes/BP_LandmassBrush.BP_LandmassBrush_C|SetDistanceIsCached false _array_element)))))

(fn PerformIterations
  (Class|ErosionComponent|ErosionButtonPress (Variables|Default|GetLandscapeErosionComponent)))

(fn RenderSeed
  (Rendering|ClearRenderTarget2D (Variables|7%29Debug|RenderTargets|GetNoise%20Seed%20%20RT))
  (Rendering|DrawMaterialtoRenderTarget (Variables|7%29Debug|RenderTargets|GetNoise%20Seed%20%20RT) (Variables|7%29Debug|MIDs|GetSeedNoiseMID)))

(fn Reset
  (2%29Generation|UpdateHeight))

(fn SetMIDParams
  (bind _weightmaps (Variables|3%29Landscape|WeightmapSetup|GetWeightmaps))
  (bind _meshpreviewmid (Variables|7%29Debug|MIDs|GetMeshPreviewMID))
  (Rendering|Material|SetVectorParameterValue (Variables|7%29Debug|MIDs|GetSeedNoiseMID) "ElevationParams" (Utilities|Struct|MakeLinearColor (Variables|2%29Generation|GetElevation%20Scale) 0.0 (Variables|2%29Generation|GetElevation%20Pre-Offset) (Math|Conversions|ToFloat(Boolean) (Variables|2%29Generation|GetInvert%20Noise))) :!class /Script/Engine.MaterialInstanceDynamic)
  (Rendering|Material|SetVectorParameterValue (Variables|7%29Debug|MIDs|GetSeedNoiseMID) "NoiseSeedParams" (Utilities|Struct|MakeLinearColor (Variables|2%29Generation|GetTiling) (Variables|2%29Generation|GetSeed) (Math|Conversions|ToFloat(Integer) (Variables|2%29Generation|GetOctaves)) (Variables|2%29Generation|GetPersistence)) :!class /Script/Engine.MaterialInstanceDynamic)
  (Rendering|Material|SetVectorParameterValue (Variables|7%29Debug|MIDs|GetSeedNoiseMID) "TestFeatureParams" (Utilities|Struct|MakeLinearColor (Variables|2%29Generation|TestFeatures|GetRamp) (Variables|2%29Generation|TestFeatures|GetPlateau) (Variables|2%29Generation|TestFeatures|GetChannel) (Variables|2%29Generation|TestFeatures|GetEdge%20Lip)) :!class /Script/Engine.MaterialInstanceDynamic)
  (Rendering|Material|SetTextureParameterValue _meshpreviewmid "NormalRT" (Variables|7%29Debug|RenderTargets|GetNormal))
  (Rendering|Material|SetTextureParameterValue _meshpreviewmid "SceneCaptureRT" (Variables|7%29Debug|RenderTargets|GetMesh%20Depth%20RT))
  (Rendering|Material|SetTextureParameterValue (Variables|7%29Debug|MIDs|GetCombineLandscapeNoiseWaterMID) "SeedRT" (Variables|7%29Debug|RenderTargets|GetNoise%20Seed%20%20RT))
  (Rendering|Material|SetTextureParameterValue (Variables|7%29Debug|MIDs|GetCombineLandscapeNoiseWaterMID) "LandscapeRT" (Variables|7%29Debug|RenderTargets|GetLandscape%20Height%20RT))
  (bind _keys (Utilities|Map|Keys _weightmaps))
  (for _array_element _keys
    (if (.ReturnValue (Utilities|Map|Find _weightmaps _array_element))
      (Utilities|IsValid (Utilities|Struct|BreakRenderTargetChannels (Utilities|Map|Find _weightmaps _array_element))
        (:Is%20Valid
          (Rendering|Material|SetTextureParameterValue (Variables|7%29Debug|MIDs|GetMeshPreviewMID) _array_element (Utilities|Struct|BreakRenderTargetChannels (Utilities|Map|Find _weightmaps _array_element))))))))

(fn SetMPCParams
  (bind _totalsize (Variables|7%29Debug|GetTotalSize))
  (bind (_x _y) (Math|Vector2D|BreakVector2D _totalsize))
  (bind _overall_resolution (Variables|3%29Landscape|GetOverall%20Resolution))
  (bind (_x_2 _y_2) (Math|Vector2D|BreakVector2D _overall_resolution))
  (Rendering|Material|SetScalarParameterValue "/Landmass/Landscape/BlueprintBrushes/MPC/MPC_Erosion.MPC_Erosion" "ResX" _x_2 :!class /Script/Engine.KismetMaterialLibrary)
  (Rendering|Material|SetScalarParameterValue "/Landmass/Landscape/BlueprintBrushes/MPC/MPC_Erosion.MPC_Erosion" "ResY" _y_2 :!class /Script/Engine.KismetMaterialLibrary)
  (Rendering|Material|SetScalarParameterValue "/Landmass/Landscape/BlueprintBrushes/MPC/MPC_Erosion.MPC_Erosion" "GridSize" (Math|Vector2D|BreakVector2D (/ (Variables|7%29Debug|GetTotalSize) (- (Variables|3%29Landscape|GetOverall%20Resolution) 1.0))) :!class /Script/Engine.KismetMaterialLibrary)
  (Rendering|Material|SetScalarParameterValue "/Landmass/Landscape/BlueprintBrushes/MPC/MPC_Erosion.MPC_Erosion" "ShowProcedural" (Math|Conversions|ToFloat(Boolean) (Variables|3%29Landscape|GetShow%20Procedural%20Impact)) :!class /Script/Engine.KismetMaterialLibrary)
  (Rendering|Material|SetScalarParameterValue "/Landmass/Landscape/BlueprintBrushes/MPC/MPC_Erosion.MPC_Erosion" "StartWater" (Variables|4%29Erosion|GetStart%20Water) :!class /Script/Engine.KismetMaterialLibrary)
  (Rendering|Material|SetScalarParameterValue "/Landmass/Landscape/BlueprintBrushes/MPC/MPC_Erosion.MPC_Erosion" "UseWaterLevel" (Math|Conversions|ToFloat(Boolean) (Variables|4%29Erosion|GetWater%20is%20Level)) :!class /Script/Engine.KismetMaterialLibrary)
  (Rendering|Material|SetScalarParameterValue "/Landmass/Landscape/BlueprintBrushes/MPC/MPC_Erosion.MPC_Erosion" "ShowWater" (Math|Conversions|ToFloat(Boolean) (and (Variables|1%29RealTimeSimulation|GetShow%20Water) (Utilities|Enum|NotEqual(Enum) (Variables|4%29Erosion|GetErosion%20Mode) "NewEnumerator0"))) :!class /Script/Engine.KismetMaterialLibrary)
  (Rendering|Material|SetScalarParameterValue "/Landmass/Landscape/BlueprintBrushes/MPC/MPC_Erosion.MPC_Erosion" "ShowComponents" (Math|Conversions|ToFloat(Boolean) (Variables|7%29Debug|GetHighlight%20Components)) :!class /Script/Engine.KismetMaterialLibrary)
  (Rendering|Material|SetScalarParameterValue "/Landmass/Landscape/BlueprintBrushes/MPC/MPC_Erosion.MPC_Erosion" "TotalSizeX" _x :!class /Script/Engine.KismetMaterialLibrary)
  (Rendering|Material|SetScalarParameterValue "/Landmass/Landscape/BlueprintBrushes/MPC/MPC_Erosion.MPC_Erosion" "TotalSizeY" _y :!class /Script/Engine.KismetMaterialLibrary)
  (Rendering|Material|SetScalarParameterValue "/Landmass/Landscape/BlueprintBrushes/MPC/MPC_Erosion.MPC_Erosion" "ComponentRes" (Utilities|Select :Index (Variables|3%29Landscape|GetComponent%20Size) :NewEnumerator0 128.0 :NewEnumerator1 256.0) :!class /Script/Engine.KismetMaterialLibrary)
  (Rendering|Material|SetScalarParameterValue "/Landmass/Landscape/BlueprintBrushes/MPC/MPC_Erosion.MPC_Erosion" "ZScale" (Variables|7%29Debug|GetRaw%20Z%20Scale) :!class /Script/Engine.KismetMaterialLibrary)
  (Rendering|Material|SetScalarParameterValue "/Landmass/Landscape/BlueprintBrushes/MPC/MPC_Erosion.MPC_Erosion" "deltaT" (Variables|4%29Erosion|GetDelta%20T) :!class /Script/Engine.KismetMaterialLibrary)
  (Rendering|Material|SetScalarParameterValue "/Landmass/Landscape/BlueprintBrushes/MPC/MPC_Erosion.MPC_Erosion" "RainAmount" (Variables|4%29Erosion|GetRain%20Amount) :!class /Script/Engine.KismetMaterialLibrary)
  (Rendering|Material|SetScalarParameterValue "/Landmass/Landscape/BlueprintBrushes/MPC/MPC_Erosion.MPC_Erosion" "Solubility" (Variables|4%29Erosion|GetSolubility) :!class /Script/Engine.KismetMaterialLibrary)
  (Rendering|Material|SetScalarParameterValue "/Landmass/Landscape/BlueprintBrushes/MPC/MPC_Erosion.MPC_Erosion" "Evaporation" (Variables|4%29Erosion|GetEvaporation) :!class /Script/Engine.KismetMaterialLibrary)
  (Rendering|Material|SetScalarParameterValue "/Landmass/Landscape/BlueprintBrushes/MPC/MPC_Erosion.MPC_Erosion" "AdvectionErosion" (Variables|4%29Erosion|GetVelocity%20Advection%20Erosion) :!class /Script/Engine.KismetMaterialLibrary)
  (Rendering|Material|SetScalarParameterValue "/Landmass/Landscape/BlueprintBrushes/MPC/MPC_Erosion.MPC_Erosion" "NormalAdvection" (Variables|4%29Erosion|GetNormal%20Advection%20Erosion) :!class /Script/Engine.KismetMaterialLibrary)
  (Rendering|Material|SetScalarParameterValue "/Landmass/Landscape/BlueprintBrushes/MPC/MPC_Erosion.MPC_Erosion" "SlopeSmoothThreshold" (select (== (Variables|4%29Erosion|GetSmooth%20Slopes%20Greater%20Than) 0.0) 999.0 (Variables|4%29Erosion|GetSmooth%20Slopes%20Greater%20Than)) :!class /Script/Engine.KismetMaterialLibrary)
  (Rendering|Material|SetScalarParameterValue "/Landmass/Landscape/BlueprintBrushes/MPC/MPC_Erosion.MPC_Erosion" "FlatSmoothThreshold" (select (== (Variables|4%29Erosion|GetSmooth%20Slopes%20Less%20Than) 0.0) 0.0 (Variables|4%29Erosion|GetSmooth%20Slopes%20Less%20Than)) :!class /Script/Engine.KismetMaterialLibrary)
  (Rendering|Material|SetScalarParameterValue "/Landmass/Landscape/BlueprintBrushes/MPC/MPC_Erosion.MPC_Erosion" "VelocityDisplay" (Variables|1%29RealTimeSimulation|GetVelocity%20Debug%20Brightness) :!class /Script/Engine.KismetMaterialLibrary)
  (Rendering|Material|SetScalarParameterValue "/Landmass/Landscape/BlueprintBrushes/MPC/MPC_Erosion.MPC_Erosion" "VelocityDamping" (Variables|4%29Erosion|GetVelocity%20Damping) :!class /Script/Engine.KismetMaterialLibrary)
  (Rendering|Material|SetScalarParameterValue "/Landmass/Landscape/BlueprintBrushes/MPC/MPC_Erosion.MPC_Erosion" "MaxAdvectSlope" (select (== (Variables|4%29Erosion|GetMax%20Advection%20Slope) 0.0) 999.0 (Variables|4%29Erosion|GetMax%20Advection%20Slope)) :!class /Script/Engine.KismetMaterialLibrary))

(fn SpawnInstanceMeshes
  (bind _landscape_ismc (Variables|Default|GetLandscape%20ISMC))
  (bind _staticmesh (Variables|/Script/Engine.StaticMeshComponent|GetStaticMesh _landscape_ismc))
  (bind _returnvalue (StaticMesh|GetBounds _staticmesh))
  (bind (_origin _boxextent) (Math|BoxSphereBounds|BreakBoxSphereBounds _returnvalue))
  (bind _y (.y _boxextent))
  (bind _returnvalue_2 (* _y 2.0))
  (bind _sections_y (Variables|3%29Landscape|GetSections%20Y))
  (bind _returnvalue_3 (Math|Conversions|ToFloat(Integer) _sections_y))
  (bind _totalsize (Variables|7%29Debug|GetTotalSize))
  (bind (_x _y_2) (Math|Vector2D|BreakVector2D _totalsize))
  (bind _returnvalue_4 (/ _y_2 _returnvalue_3))
  (bind _returnvalue_5 (/ _returnvalue_4 _returnvalue_2))
  (bind _x_2 (.x _boxextent))
  (bind _returnvalue_6 (* _x_2 2.0))
  (bind _sections_x (Variables|3%29Landscape|GetSections%20X))
  (bind _returnvalue_7 (Math|Conversions|ToFloat(Integer) _sections_x))
  (bind _returnvalue_8 (/ _x _returnvalue_7))
  (bind _returnvalue_9 (/ _returnvalue_8 _returnvalue_6))
  (bind _returnvalue_10 (Math|Vector|MakeVector _returnvalue_9 _returnvalue_5 1.0))
  (bind _returnvalue_11 (Transformation|GetActorLocation))
  (bind (_x_3 _y_3) (Math|Vector2D|BreakVector2D (Variables|7%29Debug|GetTotalSize)))
  (Components|InstancedStaticMesh|ClearInstances (Variables|Default|GetLandscape%20ISMC))
  (Variables|7%29Debug|CachedValues|SetStoredInstances (* (Variables|3%29Landscape|GetSections%20X) (Variables|3%29Landscape|GetSections%20Y)))
  (bind (_current_point _total_points _0_1_phase _poly_center) (Utilities|2DGridExecutionMacro :Sector%20Size (/ (Math|Float|Max(Float) _x_3 _y_3) (Math|Float|Max(Float) (Math|Conversions|ToFloat(Integer) (Variables|3%29Landscape|GetSections%20X)) (Math|Conversions|ToFloat(Integer) (Variables|3%29Landscape|GetSections%20Y)))) :X (Variables|3%29Landscape|GetSections%20X) :Y (Variables|3%29Landscape|GetSections%20Y))
    (:Per%20Poly%20Execution
      (Components|InstancedStaticMesh|AddInstance (Variables|Default|GetLandscape%20ISMC) (Math|Transform|MakeTransform (+ (* _poly_center "1.000000,1.000000, 0") _returnvalue_11) "0, 0, 0" _returnvalue_10) true)
      (if (Utilities|Enum|Equal(Enum) (Variables|4%29Erosion|GetErosion%20Mode) "NewEnumerator0")
        (Components|InstancedStaticMesh|AddInstance (Variables|Default|GetScatter%20Particles%20ISMC) (Math|Transform|MakeTransform (+ (* _poly_center "1.000000,1.000000, 0") _returnvalue_11) "0, 0, 0" _returnvalue_10) true)))))

(fn SpawnScatterMesh)

(fn StartTick
  (Default|SetEditorTickEnabled true)
  (Variables|7%29Debug|SetTicking%20Erosion true)
  (Rendering|SetVisibility (Variables|Default|GetActor%20Ticking) true false))

(fn StopTick
  (Default|SetEditorTickEnabled false)
  (Variables|7%29Debug|SetTicking%20Erosion false)
  (Rendering|SetVisibility (Variables|Default|GetActor%20Ticking) false false))

(fn Update
  (2%29Generation|UpdateHeight))

(fn Update
  (2%29Generation|UpdateHeight))

(fn UpdateBrushesAndNormals
  (bind (_source_rt _target_rt) (Utilities|HeightPingPong))
  (bind _landscapeerosioncomponent (Variables|Default|GetLandscapeErosionComponent))
  ;[0.2299 0.3619 0.4323] Clear Render Targets
  (Rendering|ClearRenderTarget2D _source_rt)
  (Rendering|ClearRenderTarget2D _target_rt)
  (Rendering|ClearRenderTarget2D (Variables|7%29Debug|RenderTargets|GetJump%20Flood%20Target%20A_Vel) "(R=0.000000,G=0.000000,B=0.000000,A=0.000000)")
  (Rendering|ClearRenderTarget2D (Variables|7%29Debug|RenderTargets|GetJump%20Flood%20Target%20B_Vel) "(R=0.000000,G=0.000000,B=0.000000,A=0.000000)")
  (Rendering|ClearRenderTarget2D (Variables|7%29Debug|RenderTargets|GetMesh%20Depth%20RT))
  ;[0.1309 0.3281 0.2437] Reset Simulation Index
  (Variables|7%29Debug|SetStackHeightIdx 0) ; BP_Landscape_LayerStack sets this
  (Variables|/Landmass/Landscape/OldPrototype_BP/Components/ErosionComponent.ErosionComponent_C|SetPositionIdx 0 _landscapeerosioncomponent)
  (Variables|/Landmass/Landscape/OldPrototype_BP/Components/ErosionComponent.ErosionComponent_C|SetVelocityIdx 0 _landscapeerosioncomponent)
  (Rendering|DrawMaterialtoRenderTarget (Utilities|HeightPingPong) (Variables|7%29Debug|MIDs|GetCombineLandscapeNoiseWaterMID))
  (bind _values (Utilities|Map|Values (Variables|3%29Landscape|WeightmapSetup|GetWeightmaps)))
  (for _array_element _values
    (Rendering|ClearRenderTarget2D (Utilities|Struct|BreakRenderTargetChannels _array_element)))
  (CallFunction|UpdateBrushes)
  (CallFunction|UpdateNormalandPreviewRTAssignments "")
  (if (Utilities|Enum|Equal(Enum) (Variables|4%29Erosion|GetErosion%20Mode) "NewEnumerator0")
    (Class|ErosionComponent|SeedParticlePositions (Variables|Default|GetLandscapeErosionComponent))
    (Rendering|ClearRenderTarget2D (Variables|7%29Debug|RenderTargets|GetJump%20Flood%20Target%20A_Vel) "(R=0.000000,G=0.000000,B=0.000000,A=0.000000)")
    (Rendering|ClearRenderTarget2D (Variables|7%29Debug|RenderTargets|GetJump%20Flood%20Target%20B_Vel) "(R=0.000000,G=0.000000,B=0.000000,A=0.000000)")))

(fn UpdateHeight
  (2%29Generation|RenderSeed)
  (2%29Generation|UpdateBrushesandNormals))

(fn UpdateBrushes
  (bind _layer_stack (Variables|6%29BrushLayerStack|GetLayer%20Stack))
  (bind _returnvalue (Utilities|Array|LastIndex _layer_stack))
  (bind _output_get (LocalVariables|SetInitialLastIdx (Utilities|Array|LastIndex (Variables|6%29BrushLayerStack|GetLayer%20Stack))))
  (for _index (range (+ _output_get 1))
    (Utilities|IsValid (Utilities|Array|Get(acopy) (Variables|6%29BrushLayerStack|GetLayer%20Stack) _index)
      (:Is%20Not%20Valid
        (Utilities|Array|RemoveIndex (Variables|6%29BrushLayerStack|GetLayer%20Stack) _index))))
  ;[1 1 1] Iterate Backwards so that 0 is the 'top' of the stack
  (for _index_2 (range (+ _returnvalue 1))
    (Class|BPLandmassBrush|DrawBrush (Utilities|Array|Get(acopy) (Variables|6%29BrushLayerStack|GetLayer%20Stack) (- _returnvalue _index_2)))))

(fn UpdateNormalAndPreviewRTAssignments
  (params (VelocityRT :type Texture%20Render%20Target%202D%20Object%20Reference))
  (bind _normal (Variables|7%29Debug|RenderTargets|GetNormal))
  (Utilities|IsValid VelocityRT
    (:Is%20Valid
      (Rendering|Material|SetTextureParameterValue (Variables|7%29Debug|MIDs|GetMeshPreviewMID) "VelocityRT" VelocityRT))
    (:Is%20Not%20Valid
      (Rendering|Material|SetTextureParameterValue (Variables|7%29Debug|MIDs|GetMeshPreviewMID) "VelocityRT" "/Landmass/Landscape/OldPrototype_BP/Textures/LS_Debug_RT_01.LS_Debug_RT_01")))
  (Rendering|Material|SetTextureParameterValue (Variables|7%29Debug|MIDs|GetMeshPreviewMID) "HeightRT" (Utilities|HeightPingPong))
  (Rendering|ClearRenderTarget2D _normal)
  (Rendering|DrawMaterialtoRenderTarget _normal (Variables|7%29Debug|MIDs|GetCompute%20Normal%20MID))
  (Utilities|IsValid (Variables|3%29Landscape|GetLandscape)
    (:Is%20Valid
      (bind _outactors (Actor|GetAllActorsOfClass "/Script/Landscape.LandscapeProxy"))
      (for _array_element _outactors
        (Landscape|Runtime|Material|SetLandscapeMaterialTextureParameterValue _array_element "StackRT" (Utilities|HeightPingPong))
        (Landscape|Runtime|Material|SetLandscapeMaterialTextureParameterValue _array_element "LandscapeRT" (Variables|7%29Debug|RenderTargets|GetLandscape%20Height%20RT))
        (Landscape|Runtime|Material|SetLandscapeMaterialTextureParameterValue _array_element "NormalRT" (Variables|7%29Debug|RenderTargets|GetNormal))))))

(fn ConstructionScript
  (CallFunction|Constructor))

(fn WriteLandscapeWeightmap
  (Rendering|LandscapeImportWeightmapfromRenderTarget (Variables|3%29Landscape|GetLandscape) (Variables|7%29Debug|RenderTargets|GetRender%20Target%20A) (Variables|3%29Landscape|GetLayer%20Weight%20Name) 0))
