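; BP_ScreenOverlay is a prototype Landmass landscape Blueprint that controls a screen overlay through a post-process component.
; Its construction script writes the Scale and Opacity variables to the OverlaySize and OverlayOpacity scalar parameters of the MPC_Drawing material parameter collection.
; The Enable variable toggles the PostProcess component on or off.
; BeginPlay, overlap and Tick events are empty, so all logic runs at construction.
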(event EventBeginPlay)

(event Collision|EventActorBeginOverlap
  (params (OtherActor :type Actor%20Object%20Reference)))

(event EventTick
  (params (DeltaSeconds :type Float%20%28single-precision%29)))

(fn ConstructionScript
  (Rendering|Material|SetScalarParameterValue "/Landmass/Landscape/OldPrototype_BP/Mats/MPC/MPC_Drawing.MPC_Drawing" "OverlaySize" (Variables|Default|GetScale) :!class /Script/Engine.KismetMaterialLibrary)
  (Rendering|Material|SetScalarParameterValue "/Landmass/Landscape/OldPrototype_BP/Mats/MPC/MPC_Drawing.MPC_Drawing" "OverlayOpacity" (Variables|Default|GetOpacity) :!class /Script/Engine.KismetMaterialLibrary)
  (Variables|/Script/Engine.PostProcessComponent|SetbEnabled (Variables|Default|GetEnable) (Variables|Default|GetPostProcess)))
